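# Niagara system for a Season 20 front-end wind/fog effect built on a Niagara Fluids 2D gas simulation, with both emitters saved disabled.
# Chill_Source spawns GPU particles in a user-driven box, moves them with noise and vortex forces, and assigns gas density, temperature, radius and velocity scale.
# Frosty_Fog reads those particles into camera-facing 2D simulation, pressure and lighting grids, with render targets and two sprite renderers.
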NiagaraSystem /Game/Effects/Fort_Effects/FrontEnd/S20/NS_Wind_Fluid_FrontEnd
dependencies {
 usedDataInterfaces [
  /Script/Niagara.NiagaraDataInterfaceGrid2DCollection
  /Script/Niagara.NiagaraDataInterfaceRenderTarget2D
  &NiagaraDataInterfaceParticleRead
  &NiagaraDataInterfaceCamera
  &NiagaraDataInterfaceColorCurve
  &NiagaraDataInterfacePlatformSet
  /Script/Niagara.NiagaraDataInterfaceEmitterProperties
  &NiagaraDataInterfaceCurve
 ]
 usedDynamicInputs []
 usedModules [&SystemState]
}
emitters {
 Chill_Source {
  bEnabled=false
  simTarget=GPUComputeSim
  emitterUpdateScript [
   /NiagaraFN/Modules/Scalability/SignificantInstancesCull(Cull Response:ENiagaraExecutionStateManagement=::SleepAndLetParticlesFinish, First Frame Only=false)
   &EmitterState(
    Life Cycle Mode=::NewEnumerator1
    Loop Duration=&RandomRangeFloat(Minimum=2, Maximum=5)
    Recalculate Duration Each Loop=true
    Scalability Mode=::NewEnumerator1
    UseLoopDelay=false
   )
   &SpawnRate(
    SpawnRate=&Multiply_Float(
     A=&FloatFromCurve(
      FloatCurve=DataInterface() {propertyValues {Curve {postInfinityExtrap=RCCE_Cycle, preInfinityExtrap=RCCE_Cycle, keys [<at 0 1>,<at 0.5 0.7>,<at 1 1>]}}}
      CurveIndex=$Emitter.NormalizedLoopAge
      Scale Curve=15
     )
     B=$User.SpawnScale
    )
    Use Spawn Probability=false
   )
   &SpawnBurst_Instantaneous(Spawn Count=100, Use Spawn Probability=false, UseLoopCountLimit=true)
  ]
  particleSpawnScript [
   /Niagara/Modules/Spawn/Initialization/InitializeParticle(
    Color=<rgba 1 0 0 1>
    Lifetime=&RandomRangeFloat(Minimum=9, Maximum=12)
    Mass=&RandomRangeFloat(Minimum=0.5, Maximum=1)
    Module.Write Color=true
    Module.Write Lifetime=true
    Module.Write Mass=true
    Module.Write Material Random=true
    Module.Write Position=false
    Module.Write Scale=false
    Module.Write SpriteRotation=false
    Module.Write SpriteSize=true
    Sprite Size=<0 0>
   )
   &BoxLocation(
    Box Offset=$User.Box Offset
    Box Origin:NiagaraPosition=/Niagara/DynamicInputs/Helpers/SimulationPosition()
    Box Size=$User.BoxSize
    MaskSpawnGroup=false
    Module.Override Seed=false
    Rotate Box=true
    Rotate Box On Axis:Vector3f=<0 1 0>
    Surface Only=false
   )
   &AddVelocity(Velocity=&RandomRangeVector(Minimum=<10 0 15>, Maximum=<25 0 50>))
   &CurlNoiseForce(
    Curl Noise Cone Mask Angle=35
    Curl Noise Cone Mask Falloff Angle=35
    Mask Curl Noise=true
    Noise Frequency=1
    Noise Strength=25
    Pan Noise=true
    Pan Noise Field=<1 0 0>
   ) {module=CurlNoiseForce001}
   &ApplyInitialForces(Apply Force to Position=true)
  ]
  particleUpdateScript [
   &ParticleState()
   &GravityForce(Gravity=&Multiply_VectorByFloat(Vector=&RandomRangeVector(Minimum=<5 0 5>, Maximum=<15 25 15>), Float=1))
   &Drag(Drag=0.1, Use Linear Drag=true, Use Rotational Drag=false)
   &ScaleColor(
    Scale Alpha=&FloatFromCurve(FloatCurve=DataInterface() {propertyValues {Curve {keys [<at -0.0046 1.0113 RCIM_Cubic>,<at 1.0011 0.002 RCIM_Cubic>]}}})
    ScaleA=true
    ScaleRGB=true
   )
   &CurlNoiseForce(Noise Frequency=5, Noise Strength=5, Pan Noise=true, Pan Noise Field=<1 0 1>)
   &VortexForce(
    Influence Falloff Exponent=0.2
    Influence Falloff Radius=2500
    Origin Pull Amount:NiagaraFloat=&RandomRangeFloat(Minimum=25, Maximum=15)
    Use Influence Falloff=true
    UseOriginOffset=true
    Vortex Axis=<0 1 0>
    Vortex Force Amount=-50
    Vortex Origin Offset:Vector3f=<0 0 500>
   )
   &SolveForcesAndVelocity(Acceleration Limit=1, Clamp Velocity=true, Limit Acceleration=true, Speed Limit=75)
   /NiagaraFluids/Modules/Particles/Fluids_Gas_Source(
    Density:NiagaraFloat=&FloatFromCurve(FloatCurve=DataInterface() {propertyValues {Curve {
        postInfinityExtrap=RCCE_Cycle
        preInfinityExtrap=RCCE_Cycle
        keys [
         <at 0.4 0 RCIM_Cubic>
         {arriveTangent=0.625, interpMode=RCIM_Cubic, leaveTangent=0.625, time=0.5, value=0.875}
         {arriveTangent=-1.875, interpMode=RCIM_Cubic, leaveTangent=-1.875, time=0.6, value=0.125}
         <at 0.7 0.5 RCIM_Cubic>
         {arriveTangent=-0.625, interpMode=RCIM_Cubic, leaveTangent=-0.625, time=0.8, value=0.125}
         {arriveTangent=-0.625, interpMode=RCIM_Cubic, leaveTangent=-0.625, time=0.9, value=0.375}
         <at 1 0 RCIM_Cubic>
        ]
       }}})
    NoiseIntensity=3
    NoiseScale=0.45
    Radius:NiagaraFloat=&FloatFromCurve(
     FloatCurve=DataInterface() {propertyValues {Curve {
        postInfinityExtrap=RCCE_Cycle
        preInfinityExtrap=RCCE_Cycle
        keys [
         {interpMode=RCIM_Cubic, leaveTangentWeight=0.0667, tangentMode=RCTM_User, tangentWeightMode=RCTWM_WeightedBoth, time=0.1, value=0.6}
         {
          arriveTangent=2.62
          arriveTangentWeight=0.187
          interpMode=RCIM_Cubic
          leaveTangent=2.62
          leaveTangentWeight=0.187
          tangentMode=RCTM_User
          tangentWeightMode=RCTWM_WeightedBoth
          time=0.2
          value=0.85
         }
         {
          arriveTangentWeight=0.0667
          interpMode=RCIM_Cubic
          leaveTangentWeight=0.0333
          tangentMode=RCTM_User
          tangentWeightMode=RCTWM_WeightedBoth
          time=0.4
          value=0.45
         }
         {
          arriveTangent=-2.0038
          arriveTangentWeight=0.0747
          interpMode=RCIM_Cubic
          leaveTangent=-2.0038
          leaveTangentWeight=0.0747
          tangentMode=RCTM_User
          tangentWeightMode=RCTWM_WeightedBoth
          time=0.5
          value=0.9
         }
         {
          arriveTangentWeight=0.0333
          interpMode=RCIM_Cubic
          leaveTangentWeight=0.0667
          tangentMode=RCTM_User
          tangentWeightMode=RCTWM_WeightedBoth
          time=0.6
          value=0.45
         }
         {arriveTangent=0.5, arriveTangentWeight=0.0667, interpMode=RCIM_Cubic, leaveTangent=0.5, leaveTangentWeight=0.0667, time=0.7, value=0.7}
         {arriveTangent=1, interpMode=RCIM_Cubic, leaveTangent=1, time=0.8, value=0.55}
         {arriveTangent=0.25, interpMode=RCIM_Cubic, leaveTangent=0.25, time=0.9, value=0.9}
         {arriveTangentWeight=0.0667, interpMode=RCIM_Cubic, time=1, value=0.6}
        ]
       }}}
     Scale Curve=&RandomRangeFloat(Minimum=55, Maximum=35)
    )
    Temperature:NiagaraFloat=&FloatFromCurve(
     FloatCurve=DataInterface() {propertyValues {Curve {keys [
         {arriveTangent=0.0141, interpMode=RCIM_Cubic, leaveTangent=0.0141, tangentMode=RCTM_User, time=0, value=0}
         <at 0.2 0.0024>
         <at 0.4 0.0009>
         <at 0.6 0.003>
         <at 0.8 0.0017>
         {arriveTangent=-0.0167, interpMode=RCIM_Cubic, leaveTangent=-0.0167, tangentMode=RCTM_User, time=1, value=0}
        ]}}}
     Scale Curve=&RandomRangeFloat(Minimum=-2, Maximum=2)
    )
    VelocityScale:NiagaraFloat=&RandomRangeFloat(Minimum=1, Maximum=5)
   )
   /SetVariables(Particles.SpriteSize=&Vector2DFromFloat(Value=&Multiply_Float(A=$Particles.fluids_source_radius, B=2))) {isSetParameters=true}
  ]
  rendererClasses [&NiagaraSpriteRendererProperties]
  renderers [{rendererClass=&NiagaraSpriteRendererProperties}]
 }
 Frosty_Fog {
  bEnabled=false
  simTarget=GPUComputeSim
  emitterSpawnScript [
   /SetVariables(Emitter.ResX=100, Emitter.ResY=100) {isSetParameters=true}
   /SetVariables(Emitter.SimGrid=DataInterface() {propertyValues {
      ClearBeforeNonIterationStage=true
      NumAttributes=0
      NumCellsMaxAxis=3
      NumCellsX=1
      NumCellsY=1
      OverrideBufferFormat=Float
      PreviewAttribute=None
      SetGridFromMaxAxis=false
      bOverrideFormat=false
      bPreviewGrid=false
      RenderTargetUserParameter {parameter {name=None, typeDefHandle {registeredTypeIndex=1}, varData []}}
      WorldBBoxSize=<5000 5000>
     }}) {isSetParameters=true}
   /SetVariables(Emitter.LightingGrid=DataInterface() {propertyValues {
      ClearBeforeNonIterationStage=true
      NumAttributes=3
      NumCellsMaxAxis=3
      NumCellsX=128
      NumCellsY=128
      OverrideBufferFormat=Float
      PreviewAttribute=None
      SetGridFromMaxAxis=false
      bOverrideFormat=false
      bPreviewGrid=false
      RenderTargetUserParameter {parameter {name=None, typeDefHandle {registeredTypeIndex=4}, varData []}}
      WorldBBoxSize=<100 100>
     }}) {isSetParameters=true}
   /SetVariables(Emitter.LightingGridDownsample=DataInterface() {propertyValues {
      ClearBeforeNonIterationStage=true
      NumAttributes=1
      NumCellsMaxAxis=3
      NumCellsX=64
      NumCellsY=64
      OverrideBufferFormat=Float
      PreviewAttribute=None
      SetGridFromMaxAxis=false
      bOverrideFormat=false
      bPreviewGrid=false
      RenderTargetUserParameter {parameter {name=None, typeDefHandle {registeredTypeIndex=4}, varData []}}
      WorldBBoxSize=<100 100>
     }}) {isSetParameters=true}
   /SetVariables() {isSetParameters=true}
   /NiagaraFluids/Modules/Grid2D/Grid2D_SetResolution(Grid:NiagaraDataInterfaceGrid2DCollection=$Emitter.SimGrid, NumCellsX=512, NumCellsY=256, WorldGridExtents:Vector2f=<2000 1000>)
   &Grid2D_SetResolution(
    Grid:NiagaraDataInterfaceGrid2DCollection=$Emitter.PressureGrid
    OtherGrid:NiagaraDataInterfaceGrid2DCollection=$Emitter.SimGrid
    Resolution Method:ENiagaraGrid2DResolution=::NewEnumerator3
    WorldGridExtents:Vector2f=$Emitter.Grid2D_SetResolution.WorldGridExtents
   ) {module=Grid2D_SetResolution001}
   &Grid2D_SetResolution(
    Grid:NiagaraDataInterfaceGrid2DCollection=$Emitter.LightingGridDownsample
    OtherGrid:NiagaraDataInterfaceGrid2DCollection=$Emitter.SimGrid
    Resolution Method:ENiagaraGrid2DResolution=::NewEnumerator3
    Resolution Mult=0.25
    WorldGridExtents:Vector2f=$Emitter.Grid2D_SetResolution.WorldGridExtents
   ) {module=Grid2D_SetResolution002}
   &Grid2D_SetResolution(
    Grid:NiagaraDataInterfaceGrid2DCollection=$Emitter.LightingGrid
    OtherGrid:NiagaraDataInterfaceGrid2DCollection=$Emitter.SimGrid
    Resolution Method:ENiagaraGrid2DResolution=::NewEnumerator3
    WorldGridExtents:Vector2f=$Emitter.Grid2D_SetResolution.WorldGridExtents
   ) {module=Grid2D_SetResolution003}
   /SetVariables(Emitter.dt=0.0167) {isSetParameters=true}
   /SetVariables(Emitter.g:Vector2f=<0 -9.8>) {isSetParameters=true}
   /SetVariables() {isSetParameters=true}
   /SetVariables(Emitter.density=1) {isSetParameters=true}
   /SetVariables(Emitter.ParticleSource=DataInterface() {propertyValues {EmitterBinding {bindingMode=Other, emitterName=Chill_Source}}}) {isSetParameters=true}
   /SetVariables() {isSetParameters=true}
   &self:Frosty_Fog_3.NiagaraScratchPadContainer_2.SetLightingRT_0(LightingGrid:NiagaraDataInterfaceGrid2DCollection=$Emitter.LightingGrid, LightingRT:NiagaraDataInterfaceRenderTarget2D=$Emitter.LightingRT) {module=SetLightingRT_29}
   /NiagaraFluids/Modules/Grid2D/Grid2D_UpdateSimAttributes(
    Grid:NiagaraDataInterfaceGrid2DCollection=$Emitter.SimGrid
    WorldGridSize:Vector2f=$Emitter.Grid2D_SetResolution.WorldGridExtents
    WorldScale=0.6
   ) {module=UpdateSimAttributes}
   /SetVariables() {isSetParameters=true}
   /SetVariables(Emitter.testCurve=DataInterface() {propertyValues {
      bExposeCurve=true
      AlphaCurve {keys []}
      BlueCurve {keys [<at 0.3032 0.2185>,<at 0.6671 0.6234>]}
      GreenCurve {keys [<at 0.3032 0>,<at 0.6671 0>]}
      RedCurve {keys [<at 0.3032 1>,<at 0.6671 1>]}
     }}) {isSetParameters=true}
   /NiagaraFluids/Functions/General/Fluids_Gas_Manager(
    Density Bouyancy=-0.4
    Density Color Curve=DataInterface() {propertyValues {
      ExposedName=Curve Density
      bExposeCurve=true
      AlphaCurve {keys [{arriveTangent=3.9623, interpMode=RCIM_Cubic, time=0, value=0}, {interpMode=RCIM_Cubic, leaveTangent=2.4288, time=0.5, value=1}]}
      BlueCurve {keys [<at 0.4052 0.5029>,<at 0.8274 0.4851>]}
      GreenCurve {keys [<at 0.4052 0.5376>,<at 0.8274 0.95>]}
      RedCurve {keys [<at 0.4052 0.956>,<at 0.8274 1>]}
     }}
    Density Scale=0.5
    Depth Fade=200
    Dissipation Density=0.05
    Dissipation Temperature=0.2
    Dissipation Velocity=0.25
    EnvironmentColor:LinearColor=<rgba 0 0.9911 0.1241 1>
    EnvironmentDirection:Vector3f=<0 0 -1>
    MultipleScatteringFactor=5
    Pressure Relaxation=1
    Radius Scale=2.1
    Shadowing=0.6
    SunColor:LinearColor=<rgba 5 1 0.0484 1>
    SunDirection:Vector3f=<0 0 1>
    Temperature Color Curve=DataInterface() {propertyValues {
      ExposedName=Curve Temperature
      bExposeCurve=true
      AlphaCurve {keys [<at 0 0>,<at 0.917 1>]}
      BlueCurve {keys [<at 0 0.0423>,<at 0.1501 0.0465>]}
      GreenCurve {keys [<at 0 0.0448>,<at 0.1501 0>]}
      RedCurve {keys [<at 0 0.0625>,<at 0.1501 0.0625>]}
     }}
    Temperature Scale=0.3
    Velocity Scale=0.2
    Vorticity Confinement=0.2
   )
   /NiagaraFluids/Modules/Grid2D/Grid2D_InitRT(Grid:NiagaraDataInterfaceGrid2DCollection=$Emitter.SimGrid, SimRT:NiagaraDataInterfaceRenderTarget2D=$Emitter.SimRT) {module=Grid2D_Gas_InitRT}
   &self:Frosty_Fog_3.NiagaraScratchPadContainer_2.ScratchModule_0(Grid:NiagaraDataInterfaceGrid2DCollection=$Emitter.SimGrid, PressureGrid:NiagaraDataInterfaceGrid2DCollection=$Emitter.PressureGrid) {module=ScratchModule}
  ]
  emitterUpdateScript [
   &SignificantInstancesCull(Cull Response:ENiagaraExecutionStateManagement=::SleepAndLetParticlesFinish, First Frame Only=false)
   &EmitterState(Life Cycle Mode=::NewEnumerator1, Loop Duration=1, Scalability Mode=::NewEnumerator1, UseLoopDelay=false)
   /SetVariables(Emitter.dt:NiagaraFloat=$Engine.DeltaTime) {isSetParameters=true}
   /NiagaraFluids/Modules/Grid2D/Grid2D_CreateUnitToWorldTransform(
    Camera:NiagaraDataInterfaceCamera=$Emitter.CameraQuery
    Camera Facing=true
    Offset:NiagaraPosition=$Engine.Owner.Position
    Rotation:Quat4f=$Engine.Owner.Rotation
    WorldGridExtents:Vector2f=$Emitter.Grid2D_SetResolution.WorldGridExtents
   )
   /SetVariables(Emitter.Position:NiagaraPosition=&ConvertVectorToPosition(
     Input Position=$Emitter.Grid2D_CreateUnitToWorldTransform.LocalToWorldTranslation
     Conversion Operation:ENiagara_LWCConvertVecToPos=::NewEnumerator0
    )) {isSetParameters=true}
   /NiagaraFluids/Modules/Grid2D/Grid2D_SetBounds(
    CameraFacing:NiagaraBool=$Emitter.Grid2D_CreateUnitToWorldTransform.CameraFacing
    EmitterProperties=DataInterface() {propertyValues {EmitterBinding {bindingMode=Self, emitterName=None}}}
    LocalPivot:Vector2f=$Emitter.Grid2D_CreateUnitToWorldTransform.LocalPivot
    Rotate for Camera Facing=true
    WorldGridExtents:Vector2f=$Emitter.Grid2D_SetResolution.WorldGridExtents
   )
  ]
  rendererClasses [&NiagaraSpriteRendererProperties]
  renderers [
   {rendererClass=&NiagaraSpriteRendererProperties}
   {rendererClass=&NiagaraSpriteRendererProperties, rendererIndex=1}
  ]
 }
}
summary {
 systemName=NS_Wind_Fluid_FrontEnd
 emitters [
  {bEnabled=false, emitterName=Frosty_Fog, simTarget=GPUComputeSim, rendererClasses [&NiagaraSpriteRendererProperties]}
  {bEnabled=false, emitterName=Chill_Source, simTarget=GPUComputeSim, rendererClasses [&NiagaraSpriteRendererProperties]}
 ]
 userVariables [
  User.Box Offset {type=Vector3f, description="", defaultValue {struct=Vector3f, value=<-600 0 -80>}}
  User.BoxSize {type=Vector3f, description="", defaultValue {struct=Vector3f, value=<150 15 124>}}
  User.SpawnScale {type=NiagaraFloat, defaultValue=1.5, description=""}
  User.SpawnSettledScale {type=NiagaraFloat, defaultValue=1, description=""}
 ]
}
